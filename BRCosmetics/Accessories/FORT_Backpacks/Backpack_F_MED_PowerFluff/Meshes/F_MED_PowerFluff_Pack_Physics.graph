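# Physics asset for the PowerFluff backpack cosmetic, defining capsule bodies and joints for a figure with head, ears, arms and legs.
# Tiny base_ anchor bodies are angularly locked to dyn_torso for head and arms, and to dyn_spin for legs.
# From these anchors run the dyn_ head, two-segment limb and ear chains, with ears attached to the head.
# Their joints, and dyn_spin's to the torso, carry individually tuned, slightly asymmetric swing angles.
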PhysicsAsset /BRCosmetics/Accessories/FORT_Backpacks/Backpack_F_MED_PowerFluff/Meshes/F_MED_PowerFluff_Pack_Physics
bodies {
 base_arm_1_l [{shapeType=Capsule, length=0.505, radius=0.505, shapeName=base_arm_1_l_capsule, rotation=<rot 0 -90 0>}]
 base_arm_1_r [{shapeType=Capsule, length=0.505, radius=0.505, shapeName=base_arm_1_r_capsule, rotation=<rot 0 -90 0>}]
 base_head [{shapeType=Capsule, length=0.505, radius=0.505, shapeName=base_head_capsule, rotation=<rot 0 -90 0>}]
 base_leg_1_l [{shapeType=Capsule, length=0.505, radius=0.505, shapeName=base_leg_1_l_capsule, rotation=<rot 0 -90 0>}]
 base_leg_1_r [{shapeType=Capsule, length=0.505, radius=0.505, shapeName=base_leg_1_r_capsule, rotation=<rot 0 -90 0>}]
 dyn_arm_1_l [{shapeType=Capsule, length=2, radius=2, shapeName=dyn_arm_1_l_capsule, center=<2 0 0>, rotation=<rot 90 0 0>}]
 dyn_arm_1_r [{shapeType=Capsule, length=2, radius=2, shapeName=dyn_arm_1_r_capsule, center=<2 0 0>, rotation=<rot 90 0 0>}]
 dyn_arm_2_l [{shapeType=Capsule, length=4, radius=2.7, shapeName=dyn_arm_2_l_capsule, center=<4 0 0>, rotation=<rot 90 0 0>}]
 dyn_arm_2_r [{shapeType=Capsule, length=4, radius=2.7, shapeName=dyn_arm_2_r_capsule, center=<4 0 0>, rotation=<rot 90 0 0>}]
 dyn_ear_1_l [{shapeType=Capsule, length=2, radius=3, shapeName=dyn_ear_1_l_capsule, center=<2 0 0>, rotation=<rot 90 0 0>}]
 dyn_ear_1_r [{shapeType=Capsule, length=2, radius=3, shapeName=dyn_ear_1_r_capsule, center=<2 0 0>, rotation=<rot 90 0 0>}]
 dyn_ear_2_l [{shapeType=Capsule, length=2, radius=2.5, shapeName=dyn_ear_2_l_capsule, center=<1 0 0>, rotation=<rot 90 0 0>}]
 dyn_ear_2_r [{shapeType=Capsule, length=2, radius=2.5, shapeName=dyn_ear_2_r_capsule, center=<1 0 0>, rotation=<rot 90 0 0>}]
 dyn_head [{shapeType=Capsule, length=1, radius=13, shapeName=dyn_head_capsule, center=<10 0 0>}]
 dyn_leg_1_l [{shapeType=Capsule, length=2, radius=2, shapeName=dyn_leg_1_l_capsule, center=<2 0 0>, rotation=<rot 90 0 0>}]
 dyn_leg_1_r [{shapeType=Capsule, length=2, radius=2, shapeName=dyn_leg_1_r_capsule, center=<2 0 0>, rotation=<rot 90 0 0>}]
 dyn_leg_2_l [{shapeType=Capsule, length=2.2, radius=2, shapeName=dyn_leg_2_l_capsule, center=<1 0 0>, rotation=<rot 90 0 0>}]
 dyn_leg_2_r [{shapeType=Capsule, length=2.2, radius=2, shapeName=dyn_leg_2_r_capsule, center=<1 0 0>, rotation=<rot 90 0 0>}]
 dyn_spin [{shapeType=Capsule, length=1, radius=4.5, shapeName=dyn_spin_capsule, center=<-1 0 0>}]
 dyn_torso [{shapeType=Capsule, length=1, radius=4, shapeName=dyn_torso_capsule, center=<-2 0 0>, rotation=<rot 90 0 0>}]
}
constraints [
 {
  bone1Name=base_head
  bone2Name=dyn_torso
  swing1LimitDegrees=45
  swing1Motion=Locked
  swing2LimitDegrees=45
  swing2Motion=Locked
  twistLimitDegrees=45
  twistMotion=Locked
 }
 {bone1Name=dyn_head, bone2Name=base_head, swing1LimitDegrees=5, swing2LimitDegrees=5, twistLimitDegrees=5}
 {bone1Name=dyn_ear_1_l, bone2Name=dyn_head, swing1LimitDegrees=3, swing2LimitDegrees=6, twistLimitDegrees=0}
 {bone1Name=dyn_ear_2_l, bone2Name=dyn_ear_1_l, swing1LimitDegrees=3, swing2LimitDegrees=10, twistLimitDegrees=0}
 {bone1Name=dyn_ear_1_r, bone2Name=dyn_head, swing1LimitDegrees=3, swing2LimitDegrees=5, twistLimitDegrees=0}
 {bone1Name=dyn_ear_2_r, bone2Name=dyn_ear_1_r, swing1LimitDegrees=5, swing2LimitDegrees=15, twistLimitDegrees=0}
 {bone1Name=dyn_spin, bone2Name=dyn_torso, swing1LimitDegrees=7, swing2LimitDegrees=10, twistLimitDegrees=0}
 {
  bone1Name=base_leg_1_l
  bone2Name=dyn_spin
  swing1LimitDegrees=45
  swing1Motion=Locked
  swing2LimitDegrees=45
  swing2Motion=Locked
  twistLimitDegrees=45
  twistMotion=Locked
 }
 {bone1Name=dyn_leg_1_l, bone2Name=base_leg_1_l, swing1LimitDegrees=30, swing2LimitDegrees=25, twistLimitDegrees=0}
 {bone1Name=dyn_leg_2_l, bone2Name=dyn_leg_1_l, swing1LimitDegrees=25, swing2LimitDegrees=25, twistLimitDegrees=0}
 {
  bone1Name=base_leg_1_r
  bone2Name=dyn_spin
  swing1LimitDegrees=45
  swing1Motion=Locked
  swing2LimitDegrees=45
  swing2Motion=Locked
  twistLimitDegrees=45
  twistMotion=Locked
 }
 {bone1Name=dyn_leg_1_r, bone2Name=base_leg_1_r, swing1LimitDegrees=35, swing2LimitDegrees=25, twistLimitDegrees=0}
 {bone1Name=dyn_leg_2_r, bone2Name=dyn_leg_1_r, swing1LimitDegrees=30, swing2LimitDegrees=30, twistLimitDegrees=0}
 {
  bone1Name=base_arm_1_l
  bone2Name=dyn_torso
  swing1LimitDegrees=45
  swing1Motion=Locked
  swing2LimitDegrees=45
  swing2Motion=Locked
  twistLimitDegrees=45
  twistMotion=Locked
 }
 {bone1Name=dyn_arm_1_l, bone2Name=base_arm_1_l, swing1LimitDegrees=10, swing2LimitDegrees=20, twistLimitDegrees=0}
 {bone1Name=dyn_arm_2_l, bone2Name=dyn_arm_1_l, swing1LimitDegrees=20, swing2LimitDegrees=30, twistLimitDegrees=0}
 {
  bone1Name=base_arm_1_r
  bone2Name=dyn_torso
  swing1LimitDegrees=45
  swing1Motion=Locked
  swing2LimitDegrees=45
  swing2Motion=Locked
  twistLimitDegrees=45
  twistMotion=Locked
 }
 {bone1Name=dyn_arm_1_r, bone2Name=base_arm_1_r, swing1LimitDegrees=10, swing2LimitDegrees=30, twistLimitDegrees=0}
 {bone1Name=dyn_arm_2_r, bone2Name=dyn_arm_1_r, swing1LimitDegrees=15, swing2LimitDegrees=15, twistLimitDegrees=0}
]
